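# Niagara system for the Inferno Vortex vehicle boost cosmetic, built from three ribbon emitters that stream backward from the car's filtered boost sockets, one ribbon per port.
# Ribbons follow skeletal socket positions plus a car offset, and inherited source movement fades over their lifetime.
# The layers differ in lifetime, width and speed, with two driving dynamic material parameters.
# A painted flag switches material mode, desaturation and brightness alongside per-variant color overrides.
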NiagaraSystem /VehicleCosmetics/Boosts/Boost_InfernoVortex/FX/NS_Boost_InfernoVortex
dependencies {
 usedDataInterfaces [&NiagaraDataInterfaceCurve, &NiagaraDataInterfaceArrayFloat, &NiagaraDataInterfaceSkeletalMesh]
 usedDynamicInputs []
 usedModules [
  &self:SystemSpawnScript.SystemScriptSource.SystemScriptGraph.NiagaraNodeAssignment_0.SetVariables_#1
  &SystemState
 ]
}
emitters {
 NE_VC_Boost_Ribbon_Parent001 {
  emitterSpawnScript [/SetVariables(Emitter.Desaturation:NiagaraFloat=&MakeCustomFloatFromBool(Bool=&InvertBool(Boolean:NiagaraBool=$User.bIsPainted), True Float=0, False Float=1)) {isSetParameters=true}]
  emitterUpdateScript [
   &EmitterState()
   &VC_FilteredSocketCount()
   &VC_RateQualityHandler(High Spawn Rate=60, Low Spawn Rate=60, Medium Spawn Rate=60, Use Custom Rate?=true) {enabled=false}
   &VC_SpawnRibbon_v1(Number of Ribbons=$Emitter.PortCount, SpawnRate=45)
  ]
  particleSpawnScript [
   &VC_PaintedVariantsOverride(
    Black=<rgba 0.0194 0.0194 0.0194 1>
    BurntSienna=<rgba 0.28 0.056 0 1>
    Cobalt=<rgba 0.04 0.1227 1 1>
    Crimson=<rgba 0.6308 0.0648 0.0802 1>
    ForestGreen=<rgba 0.2048 0.8714 0.1307 1>
    Grey=<rgba 0.3 0.3 0.3 1>
    Lime=<rgba 0.3238 0.67 0 1>
    Orange=<rgba 1 0.1529 0.0006 1>
    Pink=<rgba 0.7454 0.0976 0.8963 1>
    Purple=<rgba 0.1268 0.0452 0.9047 1>
    Saffron=<rgba 0.55 0.4247 0 1>
    SkyBlue=<rgba 0.107 0.4621 0.9301 1>
    TitaniumWhite=<rgba 0.6 0.6 0.6 1>
   )
   &VC_InitializeRibbon_v1(
    Color=$Particles.Color
    Custom Offset=true
    Enable Skeletal Position=true
    Facing Axis:ENiagaraOrientationAxis=::ZAxis
    Lifetime=0.8
    Local Vector Offset=&AddVectorToPosition(Position=<-55 0 -1>, Vector=Linked() {linkedVariable {name=User.DTCarOffset, type=NiagaraPosition}})
    Offset Amount=1
    Ribbon Width=25
   )
   &AddVelocity(Velocity=<-1000 0 0>, Velocity Speed Scale=1.5) {module=AddVelocity001}
  ]
  particleUpdateScript [
   &ParticleState()
   &ScaleRibbonWidth(Ribbon Width Scale=&FloatFromCurve(FloatCurve=DataInterface() {propertyValues {Curve {keys [<at 0 1.25 RCIM_Cubic>,<at 1 0.6>]}}}))
   &InheritSourceMovement(
    Applied Position Scale=&VectorFromFloat(Value=&FloatFromCurve(FloatCurve=DataInterface() {propertyValues {Curve {defaultValue=1, keys [{interpMode=RCIM_Cubic, tangentMode=RCTM_User, time=0, value=1}, {tangentMode=RCTM_User, time=1, value=0}]}}}))
    ApplyForce=false
    ApplyPosition=true
    ApplyPositionFalloff=false
   )
   &SolveForcesAndVelocity(Clamp Velocity=false, Limit Acceleration=false)
   &VC_UpdateRibbon_v1(Enable Skeletal Position=true)
   &DynamicMaterialParameters(
    Index 1 Param 1=&FloatFromCurve(FloatCurve=DataInterface() {propertyValues {Curve {keys [
         {arriveTangent=-13.683, interpMode=RCIM_Cubic, leaveTangent=-13.683, tangentMode=RCTM_User, time=0, value=1}
         {arriveTangent=-4.0398, time=0.05, value=-1.6429}
         {arriveTangent=5.9399, interpMode=RCIM_Cubic, leaveTangent=5.9399, time=0.6, value=-0.9}
         {arriveTangent=-0.0339, interpMode=RCIM_Cubic, leaveTangent=-0.0339, tangentMode=RCTM_User, time=1, value=4}
        ]}}})
    Module.Index0Float0Write=true
    Module.Index0Float1Write=true
    Module.Index0Float2Write=false
    Module.Index0Float3Write=false
    Module.Index1Float0Write=true
    Module.Index1Float1Write=true
    Module.Index1Float2Write=true
    Module.Index1Float3Write=true
    Write Parameter Index 1=::NewEnumerator1
   )
   &ScaleColor(
    Scale RGB=&VectorFromFloat(Value=&MakeCustomFloatFromBool(
      Bool=$User.bIsPainted
      True Float=&MakeFloatFromVector(
       Vector=&VectorFromFloat(Value=&FloatFromCurve(FloatCurve=DataInterface() {propertyValues {Curve {keys [{arriveTangent=-0.4419, interpMode=RCIM_Cubic, leaveTangent=-0.4419, tangentMode=RCTM_User, time=0, value=2}, <at 1 0 RCIM_Cubic>]}}}))
       Channel=::NewEnumerator0
      )
      False Float=1
     ))
    ScaleA=true
    ScaleRGB=true
   )
  ]
  rendererClasses [&NiagaraRibbonRendererProperties]
  renderers [{rendererClass=&NiagaraRibbonRendererProperties}]
 }
 NE_VC_Boost_Ribbon_Parent003 {
  emitterSpawnScript [
   /SetVariables(Emitter.mode selector:NiagaraFloat=&MakeCustomFloatFromBool(Bool=&InvertBool(Boolean:NiagaraBool=$User.bIsPainted), True Float=1.07, False Float=0)) {isSetParameters=true}
   /SetVariables(Emitter.Contrast:NiagaraFloat=&MakeCustomFloatFromBool(Bool=&InvertBool(Boolean:NiagaraBool=$User.bIsPainted), True Float=0, False Float=1.025)) {isSetParameters=true}
  ]
  emitterUpdateScript [
   &EmitterState(Inactive Response=::NewEnumerator1, Life Cycle Mode=::NewEnumerator1, Loop Duration=0.7, UseLoopDelay=false)
   &VC_FilteredSocketCount()
   &VC_RateQualityHandler(High Spawn Rate=60, Low Spawn Rate=60, Medium Spawn Rate=60, Use Custom Rate?=true) {enabled=false}
   &VC_SpawnRibbon_v1(Number of Ribbons=$Emitter.PortCount, Spawn Time=-1, SpawnRate=60)
  ]
  particleSpawnScript [
   &VC_PaintedVariantsOverride(
    Black=<rgba 0.014 0.014 0.014 1>
    BurntSienna=<rgba 0.16 0.032 0 1>
    Cobalt=<rgba 0.04 0.1227 1 1>
    Crimson=<rgba 0.5271 0.0545 0.0685 1>
    ForestGreen=<rgba 0.077 0.35 0.035 1>
    Grey=<rgba 0.15 0.15 0.15 1>
    Lime=<rgba 0.3973 0.8 0 1>
    Orange=<rgba 0.8469 0.1301 0.0006 1>
    Pink=<rgba 0.491 0.0648 0.5906 1>
    Purple=<rgba 0.1049 0.0264 0.8796 1>
    Saffron=<rgba 0.65 0.52 0 1>
    SkyBlue=<rgba 0.107 0.4621 0.9301 1>
    TitaniumWhite=<rgba 0.6 0.6 0.6 1>
   )
   &VC_InitializeRibbon_v1(
    Color=$Particles.Color
    Custom Offset=true
    Enable Skeletal Position=true
    Facing Axis:ENiagaraOrientationAxis=::ZAxis
    Lifetime=0.45
    Local Vector Offset=&AddVectorToPosition(Position=<-2 0 0>, Vector=Linked() {linkedVariable {name=User.DTCarOffset, type=NiagaraPosition}})
    Offset Amount=1
    Ribbon Width=20
   )
   &AddVelocity(Velocity=<-1500 0 0>, Velocity Speed Scale=1.5) {module=AddVelocity001}
  ]
  particleUpdateScript [
   &ParticleState()
   &InheritSourceMovement(
    Applied Position Scale=&VectorFromFloat(Value=&FloatFromCurve(FloatCurve=DataInterface() {propertyValues {Curve {defaultValue=1, keys [{interpMode=RCIM_Cubic, tangentMode=RCTM_User, time=0, value=1}, {tangentMode=RCTM_User, time=1, value=0}]}}}))
    ApplyForce=false
    ApplyPosition=true
    ApplyPositionFalloff=false
   )
   &ScaleRibbonWidth(Ribbon Width Scale=&FloatFromCurve(FloatCurve=DataInterface() {propertyValues {Curve {keys [
         <at 0 0 RCIM_Cubic>
         {arriveTangent=1.7429, interpMode=RCIM_Cubic, leaveTangent=1.7429, time=0.1, value=1.225}
         {arriveTangent=-1.4411, interpMode=RCIM_Cubic, leaveTangent=-1.4411, time=0.7, value=1.22}
         <at 0.95 0>
        ]}}}))
   &SolveForcesAndVelocity(Clamp Velocity=false, Limit Acceleration=false)
   &VC_UpdateRibbon_v1(Enable Skeletal Position=true)
   &DynamicMaterialParameters(
    Index 3 Param 3=&SelectFloatFromArray(Direct Array Index=&SpawnGroup(), Float Selection Array=DataInterface() {propertyValues {FloatData [1.1,0]}})
    Module.Index0Float0Write=true
    Module.Index0Float1Write=true
    Module.Index0Float2Write=true
    Module.Index0Float3Write=true
    Module.Index3Float0Write=true
    Module.Index3Float1Write=true
    Module.Index3Float2Write=true
    Module.Index3Float3Write=true
    Write Parameter Index 3=::NewEnumerator1
   )
   &ScaleColor(
    Scale RGB=&VectorFromFloat(Value=&MakeCustomFloatFromBool(
      Bool=$User.bIsPainted
      True Float=&MakeFloatFromVector(
       Vector=&VectorFromFloat(Value=&FloatFromCurve(FloatCurve=DataInterface() {propertyValues {Curve {keys [<at 0 2 RCIM_Cubic>,<at 1 0.3 RCIM_Cubic>]}}}))
       Channel=::NewEnumerator0
      )
      False Float=1
     ))
    ScaleA=true
    ScaleRGB=true
   )
  ]
  rendererClasses [&NiagaraRibbonRendererProperties]
  renderers [{rendererClass=&NiagaraRibbonRendererProperties}]
 }
 NE_VC_Boost_Ribbon_Parent004 {
  emitterSpawnScript [
   /SetVariables(Emitter.mode selector:NiagaraFloat=&MakeCustomFloatFromBool(Bool=&InvertBool(Boolean:NiagaraBool=$User.bIsPainted), True Float=1.2, False Float=0)) {isSetParameters=true}
   /SetVariables(Emitter.Contrast:NiagaraFloat=&MakeCustomFloatFromBool(Bool=&InvertBool(Boolean:NiagaraBool=$User.bIsPainted), True Float=0, False Float=1)) {isSetParameters=true}
  ]
  emitterUpdateScript [
   &EmitterState(Inactive Response=::NewEnumerator1, Life Cycle Mode=::NewEnumerator1, Loop Duration=1, UseLoopDelay=false)
   &VC_FilteredSocketCount()
   &VC_RateQualityHandler(High Spawn Rate=60, Low Spawn Rate=60, Medium Spawn Rate=60, Use Custom Rate?=true) {enabled=false}
   &VC_SpawnRibbon_v1(Number of Ribbons=$Emitter.PortCount, Spawn Time=-1, SpawnRate=60)
  ]
  particleSpawnScript [
   &VC_PaintedVariantsOverride(
    BurntSienna=<rgba 0.3 0.0638 0 1>
    Cobalt=<rgba 0.04 0.1227 1 1>
    Crimson=<rgba 0.6308 0.0648 0.0802 1>
    ForestGreen=<rgba 0.1656 0.8714 0.0871 1>
    Grey=<rgba 0.25 0.25 0.25 1>
    Lime=<rgba 0.3238 0.67 0 1>
    Orange=<rgba 1 0.305 0.0012 1>
    Pink=<rgba 0.9911 0.1301 1 1>
    Purple=<rgba 0.2016 0.0704 1 1>
    Saffron=<rgba 0.5 0.3851 0 1>
    SkyBlue=<rgba 0.107 0.4621 0.9301 1>
    TitaniumWhite=<rgba 0.5 0.5 0.5 1>
   )
   &VC_InitializeRibbon_v1(
    Color=$Particles.Color
    Custom Offset=true
    Enable Skeletal Position=true
    Facing Axis:ENiagaraOrientationAxis=::ZAxis
    Lifetime=1
    Local Vector Offset=&AddVectorToPosition(Position=<-35 0 0>, Vector=Linked() {linkedVariable {name=User.DTCarOffset, type=NiagaraPosition}})
    Offset Amount=1
    Ribbon Width=40
   )
   &AddVelocity(Velocity=<-1000 0 0>, Velocity Speed Scale=1.5) {module=AddVelocity001}
  ]
  particleUpdateScript [
   &ParticleState()
   &ScaleRibbonWidth(Ribbon Width Scale=&FloatFromCurve(FloatCurve=DataInterface() {propertyValues {Curve {keys [<at 0 1.25 RCIM_Cubic>,<at 1 0.5>]}}}))
   &InheritSourceMovement(
    Applied Position Scale=&VectorFromFloat(Value=&FloatFromCurve(FloatCurve=DataInterface() {propertyValues {Curve {defaultValue=1, keys [{interpMode=RCIM_Cubic, tangentMode=RCTM_User, time=0, value=1}, {tangentMode=RCTM_User, time=1, value=0}]}}}))
    ApplyForce=false
    ApplyPosition=true
    ApplyPositionFalloff=false
   )
   &SolveForcesAndVelocity(Clamp Velocity=false, Limit Acceleration=false)
   &VC_UpdateRibbon_v1(Enable Skeletal Position=true)
   &ScaleColor(
    Scale RGB=&VectorFromFloat(Value=&MakeCustomFloatFromBool(Bool=$User.bIsPainted, True Float=1.5, False Float=1))
    ScaleA=true
    ScaleRGB=true
   )
  ]
  rendererClasses [&NiagaraRibbonRendererProperties]
  renderers [{rendererClass=&NiagaraRibbonRendererProperties}]
 }
}
summary {
 systemName=NS_Boost_InfernoVortex
 emitters [
  {emitterName=NE_VC_Boost_Ribbon_Parent001, rendererClasses [&NiagaraRibbonRendererProperties]}
  {emitterName=NE_VC_Boost_Ribbon_Parent004, rendererClasses [&NiagaraRibbonRendererProperties]}
  {emitterName=NE_VC_Boost_Ribbon_Parent003, rendererClasses [&NiagaraRibbonRendererProperties]}
 ]
 userVariables [
  User.bIsFrontedPreview {type=NiagaraBool}
  User.bIsFrontendPreview {type=NiagaraBool}
  User.bIsPainted {type=NiagaraBool}
  User.ColorVariantID {type=NiagaraInt32}
  User.DTCarOffset {type=NiagaraPosition, description="", defaultValue {}}
  User.PaintedColor {type=LinearColor, description="", defaultValue {struct=LinearColor, value=<rgba 1 1 1 1>}}
  User.PaintedScalar {type=NiagaraFloat, defaultValue=1, description=""}
  User.PortCount {type=NiagaraInt32, defaultValue=2}
 ]
}
